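# Sound cue for pickaxe hits on Husk enemies, played with small melee attenuation and a player melee weapon concurrency set.
# Each variant mixes an immediate random husk impact with a briefly delayed blend of random enemy-hit and T04 pickaxe impact waves, both pitch-modulated.
# A quality-level root selects between a fuller variant and a matching one with fewer waves per random pool, leaving its last two slots empty.
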SoundCue /Game/Sounds/Fort_Weapons/PickAxe/PickAxe_Impact_Husk_Cue
root=QualityLevel_0
nodes [
 {id=Delay_1, children [Mixer_2], properties {delayMax=0.1, delayMin=0.1}}
 {id=Delay_2, children [Mixer_4], properties {delayMax=0.1, delayMin=0.1}}
 {id=Mixer_2, children [Modulator_2,Modulator_3], properties {inputVolume [0.5,1]}}
 {id=Mixer_3, children [Delay_1,Random_5], properties {inputVolume [1,0.7]}}
 {id=Mixer_4, children [Modulator_4,Modulator_5], properties {inputVolume [0.5,1]}}
 {id=Mixer_5, children [Delay_2,Random_8], properties {inputVolume [1,0.7]}}
 {id=Modulator_2, children [Random_3], properties {pitchMax=1.5, pitchMin=1, volumeMax=0.7, volumeMin=0.7}}
 {id=Modulator_3, children [Random_4], properties {pitchMax=1, pitchMin=0.9, volumeMax=0.7, volumeMin=0.7}}
 {id=Modulator_4, children [Random_6], properties {pitchMax=1.5, pitchMin=1, volumeMax=0.7, volumeMin=0.7}}
 {id=Modulator_5, children [Random_7], properties {pitchMax=1, pitchMin=0.9, volumeMax=0.7, volumeMin=0.7}}
 {id=QualityLevel_0, children [Mixer_3,Mixer_5,null,null]}
 {id=Random_3, children [WavePlayer_14, WavePlayer_15, WavePlayer_16, WavePlayer_17, WavePlayer_18], properties {weights [1,1,1,1,1]}}
 {id=Random_4, children [WavePlayer_19, WavePlayer_20, WavePlayer_21], properties {weights [1,1,1]}}
 {
  id=Random_5
  children [WavePlayer_22, WavePlayer_23, WavePlayer_24, WavePlayer_25, WavePlayer_26, WavePlayer_27]
  properties {weights [1,1,1,1,1,1]}
 }
 {id=Random_6, children [WavePlayer_28, WavePlayer_30, WavePlayer_31], properties {weights [1,1,1]}}
 {id=Random_7, children [WavePlayer_33, WavePlayer_34], properties {weights [1,1]}}
 {id=Random_8, children [WavePlayer_36, WavePlayer_38, WavePlayer_39, WavePlayer_40], properties {weights [1,1,1,1]}}
 {id=WavePlayer_14, properties {soundWaveAssetPtr=&/PickAxe_Enemy_001}}
 {id=WavePlayer_15, properties {soundWaveAssetPtr=&/PickAxe_Enemy_002}}
 {id=WavePlayer_16, properties {soundWaveAssetPtr=&/PickAxe_Enemy_003}}
 {id=WavePlayer_17, properties {soundWaveAssetPtr=&/PickAxe_Enemy_004}}
 {id=WavePlayer_18, properties {soundWaveAssetPtr=&/PickAxe_Enemy_005}}
 {id=WavePlayer_19, properties {soundWaveAssetPtr=&/pickaxe_impact_T04_01}}
 {id=WavePlayer_20, properties {soundWaveAssetPtr=&/pickaxe_impact_T04_02}}
 {id=WavePlayer_21, properties {soundWaveAssetPtr=&/pickaxe_impact_T04_03}}
 {id=WavePlayer_22, properties {soundWaveAssetPtr=/Game/Sounds/Fort_Monster_Husk/SFX/husk_impact_mix_ST_09}}
 {id=WavePlayer_23, properties {soundWaveAssetPtr=/Game/Sounds/Fort_Monster_Husk/SFX/husk_impact_mix_ST_13}}
 {id=WavePlayer_24, properties {soundWaveAssetPtr=/Game/Sounds/Fort_Monster_Husk/SFX/husk_impact_mix_ST_01}}
 {id=WavePlayer_25, properties {soundWaveAssetPtr=/Game/Sounds/Fort_Monster_Husk/SFX/husk_impact_mix_ST_07}}
 {id=WavePlayer_26, properties {soundWaveAssetPtr=/Game/Sounds/Fort_Monster_Husk/SFX/husk_impact_mix_ST_10}}
 {id=WavePlayer_27, properties {soundWaveAssetPtr=/Game/Sounds/Fort_Monster_Husk/SFX/husk_impact_mix_ST_04}}
 {id=WavePlayer_28, properties {soundWaveAssetPtr=&PickAxe_Enemy_001}}
 {id=WavePlayer_30, properties {soundWaveAssetPtr=&PickAxe_Enemy_003}}
 {id=WavePlayer_31, properties {soundWaveAssetPtr=&PickAxe_Enemy_004}}
 {id=WavePlayer_33, properties {soundWaveAssetPtr=&pickaxe_impact_T04_01}}
 {id=WavePlayer_34, properties {soundWaveAssetPtr=&pickaxe_impact_T04_02}}
 {id=WavePlayer_36, properties {soundWaveAssetPtr=&husk_impact_mix_ST_09}}
 {id=WavePlayer_38, properties {soundWaveAssetPtr=&husk_impact_mix_ST_01}}
 {id=WavePlayer_39, properties {soundWaveAssetPtr=&husk_impact_mix_ST_07}}
 {id=WavePlayer_40, properties {soundWaveAssetPtr=&husk_impact_mix_ST_10}}
]
properties {
 attenuationSettings=&Fort_Weapon_Melee_Sml_Att
 maxDistance=2800
 soundClassObject=&Fort_WeaponMelee_PSM
 volumeMultiplier=0.7
 concurrencySet [&Player_Melee_Weapon_Conc]
}
