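# Sound cue for pickaxe impacts, named for hits on indestructible surfaces, that plays one of five PickAxe_Enemy waves.
# An equally weighted Random node picks the wave and feeds a Modulator that applies a fixed volume cut with no pitch variation.
# The cue boosts overall volume, uses small melee weapon attenuation and the impacts sound class, and joins the player melee weapon concurrency set.
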SoundCue /Game/Sounds/Fort_Weapons/PickAxe/PickAxe_Impact_Indestructible_Surface_Cue
root=Modulator_1
nodes [
 {id=Modulator_1, children [Random_1], properties {pitchMax=1, pitchMin=1, volumeMax=0.7, volumeMin=0.7}}
 {id=Random_1, children [WavePlayer_5,WavePlayer_6,WavePlayer_7,WavePlayer_8,WavePlayer_9], properties {weights [1,1,1,1,1]}}
 {id=WavePlayer_5, properties {soundWaveAssetPtr=&/PickAxe_Enemy_001}}
 {id=WavePlayer_6, properties {soundWaveAssetPtr=&/PickAxe_Enemy_002}}
 {id=WavePlayer_7, properties {soundWaveAssetPtr=&/PickAxe_Enemy_003}}
 {id=WavePlayer_8, properties {soundWaveAssetPtr=&/PickAxe_Enemy_004}}
 {id=WavePlayer_9, properties {soundWaveAssetPtr=&/PickAxe_Enemy_005}}
]
properties {
 attenuationSettings=&Fort_Weapon_Melee_Sml_Att
 maxDistance=2800
 soundClassObject=&Fort_Impacts
 volumeMultiplier=1.5
 concurrencyOverrides {resolutionRule=PreventNew}
 concurrencySet [&Player_Melee_Weapon_Conc]
}
